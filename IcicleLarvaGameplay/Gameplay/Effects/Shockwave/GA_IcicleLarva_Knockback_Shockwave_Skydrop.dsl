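; Gameplay ability for the Icicle Larva's Skydrop shockwave knockback, overriding how the instigator of the knockback is resolved.
; GetAbilityInstigator takes the source object from the effect context that granted the ability and casts it to Actor.
; A failed cast prints an error instead of returning an instigator.
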(fn GetAbilityInstigator
  (bind _asactor (Utilities|Casting|CastToActor (Ability|EffectContext|GetSourceObject (Ability|GetGrantedbyEffectContext)))
    (:then
      (return _asactor))
    (:CastFailed
      (Development|PrintString "Error: Knockback failed to get instigator"))))
